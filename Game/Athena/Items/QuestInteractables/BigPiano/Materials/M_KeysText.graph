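# Masked, lit material for the lettering on the keys of the Big Piano quest interactable.
# A UV Offset parameter shifts a one-fifth-scale UV window over the piano keys diffuse and normal textures, apparently picking each key's glyph from an atlas.
# Big piano tile textures are world-aligned at a shared texture_scale, with the diffuse tinted by letter_tint.
# The tile normal is blended over the key normal with angle-corrected blending.
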Material /Game/Athena/Items/QuestInteractables/BigPiano/Materials/M_KeysText
graph {
 properties {blendMode=BLEND_Masked, pixelDepthOffsetMode=PDOM_Legacy, usedShadingModels=Default Lit}
 nodes [
  n1 = Add(A=TextureCoordinate() {uTiling=0.2, vTiling=0.2}, B=ComponentMask(None=VectorParameter() {parameterName=UV Offset}.RGB) {g=true, r=true})
  n2 = ScalarParameter() {defaultValue=75, parameterName=texture_scale}
  MaterialFunctionCall(
   BaseNormal=TextureSample(UVs=@n1) {samplerType=SAMPLERTYPE_Normal, texture=/Game/Athena/Items/QuestInteractables/BigPiano/Textures/T_piano_keys_01_N}.RGB
   AdditionalNormal=MaterialFunctionCall(
    TextureObject=TextureObject() {samplerType=SAMPLERTYPE_Normal, texture=/Game/Athena/Items/QuestInteractables/BigPiano/Textures/T_big_piano_tile_N}
    TextureSize=@n2
   ) {materialFunction=&WorldAlignedNormal}.XYZ Texture
  ) {materialFunction=&BlendAngleCorrectedNormals}
  MaterialFunctionCall(
   TextureObject=TextureObject() {samplerType=SAMPLERTYPE_Masks, texture=/Game/Athena/Items/QuestInteractables/BigPiano/Textures/T_big_piano_tile_S}
   TextureSize=@n2
  ) {materialFunction=&WorldAlignedTexture}
  Multiply(
   A=MaterialFunctionCall(TextureObject=TextureObject() {texture=/Game/Athena/Items/QuestInteractables/BigPiano/Textures/T_big_piano_tile_D}, TextureSize=@n2) {materialFunction=&WorldAlignedTexture}.XYZ Texture
   B=Multiply(A=ScalarParameter() {defaultValue=1, parameterName=None}, B=VectorParameter() {parameterName=letter_tint, defaultValue=<rgba 1 1 1 0>}.RGB)
  )
  ScalarParameter()
  TextureSample(UVs=@n1) {texture=/Game/Athena/Items/QuestInteractables/BigPiano/Textures/T_piano_keys_01_D}
 ]
}
