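# Parent material for the big piano quest interactable's keys, with a glow color meant to change when a key is hit.
# World-aligned tile textures, sized by a shared scale, give tinted base color, a mask whose green channel drives metallic, specular and roughness, and detail normals blended over a UV-mapped piano normal.
# Emissive is GlowColor, black by default, times the dynamic IntensityAdd, and all attributes pass through MF_CreativeMoveTool.
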Material /Game/Athena/Items/QuestInteractables/BigPiano/Materials/M_PianoKey_Parent
graph {
 properties {bUseMaterialAttributes=true, pixelDepthOffsetMode=PDOM_Legacy, usedShadingModels=Default Lit}
 nodes [
  n1 = ScalarParameter() {defaultValue=75, parameterName=texture_scale}
  n2 = MaterialFunctionCall(TextureObject=TextureObject() {texture=/Game/Athena/Items/QuestInteractables/BigPiano/Textures/T_big_piano_tile_D}, TextureSize=@n1) {materialFunction=&WorldAlignedTexture}
  n3 = VectorParameter() {parameterName=tint, defaultValue=<rgba 1 1 1 0>}
  n4 = MaterialFunctionCall(Float3=MaterialFunctionCall(
    TextureObject=TextureObject() {samplerType=SAMPLERTYPE_Masks, texture=/Game/Athena/Items/QuestInteractables/BigPiano/Textures/T_big_piano_tile_S}
    TextureSize=@n1
   ) {materialFunction=&WorldAlignedTexture}.XYZ Texture) {materialFunction=&BreakOutFloat3Components}
  n5 = Multiply(A=VectorParameter() {parameterName=GlowColor, defaultValue=<rgba 0 0 0 1>}.RGB)
  n6 = ScalarParameter() {defaultValue=2, desc=Dynamic param to change piano color when hit, parameterName=IntensityAdd}
  n7 = TextureSample() {samplerType=SAMPLERTYPE_Normal, texture=/Game/Athena/Items/QuestInteractables/BigPiano/Textures/T_big_piano_N}
  n8 = MaterialFunctionCall(
   TextureObject=TextureObject() {samplerType=SAMPLERTYPE_Normal, texture=/Game/Athena/Items/QuestInteractables/BigPiano/Textures/T_big_piano_tile_N}
   TextureSize=@n1
  ) {materialFunction=&WorldAlignedNormal}
  MaterialFunctionCall(MA In=MakeMaterialAttributes!(
    BaseColor=Multiply(A=@n2.XYZ Texture, B=@n3.RGB)
    Metallic=@n4.G
    Specular=@n4.G
    Roughness=@n4.G
    Anisotropy
    EmissiveColor=Multiply(A=@n5, B=@n6)
    Opacity
    OpacityMask
    Normal=MaterialFunctionCall(BaseNormal=@n7.RGB, AdditionalNormal=@n8.XYZ Texture) {materialFunction=&BlendAngleCorrectedNormals}.Result
    Tangent
    WorldPositionOffset
    SubsurfaceColor
    ClearCoat
    ClearCoatRoughness
    AmbientOcclusion
    Refraction
    CustomizedUVs_0
    CustomizedUVs_1
    CustomizedUVs_2
    CustomizedUVs_3
    CustomizedUVs_4
    CustomizedUVs_5
    CustomizedUVs_6
    CustomizedUVs_7
    PixelDepthOffset
    ShadingModel
    Displacement
   )) {materialFunction=&MF_CreativeMoveTool}
 ]
}
