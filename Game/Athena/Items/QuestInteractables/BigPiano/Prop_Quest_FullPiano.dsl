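; Giant Christmas-era piano quest prop built from 24 child key actors, which the construction script collects, indexes for sound and material, and spaces (white keys only).
; Outside creative mode, stepped-on keys advance a per-player map through the target note array, setting key colors and restarting or resetting on mistakes.
; Completion awards quest progress, except in playgrounds, then calls a non-authority song playback using timelines and note-name arrays to trigger keys.
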(event EventBeginPlay
  ;[0.473 1 0.441] Delay Hack
  (Utilities|FlowControl|Delay 5.0)
  ;[0.3657 0.8438 0.3207] Check if in playgrounds
  (bind _asfort_game_state_athena (Utilities|Casting|CastToFortGameStateAthena (Game|GetGameState)))
  (if (Variables|/Script/FortniteGame.FortGameStateAthena|GetbPlaylistDataIsLoaded _asfort_game_state_athena)
    (bind _cancompletequest (Class|AthenaQuestFunctionLibrary|AthenaCheckCreativeMode _asfort_game_state_athena))
    (if _cancompletequest
      (Variables|Quest-Backend|SetbIsInPlayground false)
      (else
        (Variables|Quest-Backend|SetbIsInPlayground true)))
    ;[0.5736 1 0.5619] Decide if Object should be visible to player or not depending on if they have the quest and objective or not
    (Networking|SwitchHasAuthority
      (:Remote
        (bind (_questisvalid _questcompleted _objectivecompleted) (Default|AthenaCheckQuestandObjectiveCompletion (Utilities|Array|Get(acopy) (Fortnite|GetLocalFortPlayerControllers true)) (Variables|Quest-Backend|GetRelatedInteractable_QuestItem) (Variables|Quest-Backend|GetRelatedInteractable_BackendObj)))
        (if (and _questisvalid _objectivecompleted)
          _)))
    (else
      (Playlist|BindEventtoOnCurrentPlaylistLoaded _asfort_game_state_athena (delegate Custom|PlaylistLoaded)))))

(event Custom|PlaylistLoaded
  (params (PlaylistName :type Name) (PlaylistContextTags :type Gameplay%20Tag%20Container%20Structure%20%28by%20ref%29))
  ;[0.3657 0.8438 0.3207] Check if in playgrounds
  (bind _cancompletequest (Class|AthenaQuestFunctionLibrary|AthenaCheckCreativeMode (Game|GetGameState)))
  (if _cancompletequest
    (Variables|Quest-Backend|SetbIsInPlayground false)
    (else
      (Variables|Quest-Backend|SetbIsInPlayground true)))
  ;[0.5736 1 0.5619] Decide if Object should be visible to player or not depending on if they have the quest and objective or not
  (Networking|SwitchHasAuthority
    (:Remote
      (bind (_questisvalid _questcompleted _objectivecompleted) (Default|AthenaCheckQuestandObjectiveCompletion (Utilities|Array|Get(acopy) (Fortnite|GetLocalFortPlayerControllers true)) (Variables|Quest-Backend|GetRelatedInteractable_QuestItem) (Variables|Quest-Backend|GetRelatedInteractable_BackendObj)))
      (if (and _questisvalid _objectivecompleted)
        _))))

(event Custom|KeySteppedOn
  (params (OverlappedActor :type Actor%20Object%20Reference) (OtherActor :type Actor%20Object%20Reference))
  (if (not (Variables|Default|GetbIsCreativeMode))
    ;[0.473 1 0.441] Set key we hit and get the name of it, then get the player on top of the key
    (Utilities|FlowControl|Sequence
      (:then_0
        (bind _asprop_quest_piano_key (Utilities|Casting|CastToProp_Quest_PianoKey OverlappedActor))
        (bind _output_get (Variables|Quest-BP|SetKeyHit _asprop_quest_piano_key))
        (Variables|Quest-BP|SetKeyNameSteppedOn (Variables|/Game/Athena/Items/QuestInteractables/BigPiano/Prop_Quest_PianoKey.Prop_Quest_PianoKey_C|GetKeyName _output_get)))
      (:then_1
        (Variables|Quest-BP|SetPlayerOnKey (Fortnite|GetFortPlayerControllerfromActor OtherActor))
        (CallFunction|TrackQuestProgress 0)))))

(event Custom|TrackQuestProgress
  (params (Player :type Fort%20Player%20Controller%20Object%20Reference) (KeyHit :type Actor%20Object%20Reference))
  (bind _playeronkey (Variables|Quest-BP|GetPlayerOnKey))
  ;[0.473 1 0.441] Reset progress if no one touches keys after 10 seconds
  (Utilities|Time|SetTimerbyEvent (EventDispatchers|CreateEvent :!event Custom|ClearPianoQuestProgress) 15.0)
  ;[0.473 1 0.441] This function will find & set bIsFirstKey & bIsKeyInQuest
  (bind (_ispartofquest _isfirstkey) (CallFunction|IsKeyPartOfQuest (Variables|Quest-BP|GetKeyNameSteppedOn)))
  (Variables|Quest-BP|SetbIsKeyInQuest _ispartofquest)
  (Variables|Quest-BP|SetbIsFirstKey _isfirstkey)
  ;[0.473 1 0.441] Don't continue if quest progress is 0 & players don't step on the first key
  (if (not (and (== (Utilities|Map|Find (Variables|Quest-BP|GetPlayerQuestTracking) (Variables|Quest-BP|GetPlayerOnKey)) 0) (not _isfirstkey)))
    ;[0.5358 1 0.4366] Searches for passed runner index & returns their quest progress
    (bind (_questprogress _hitcorrectnote) (CallFunction|IsHitKeyCorrect))
    ;[0.473 1 0.441] Did they hit the right key?
    (if _hitcorrectnote
      ;[0.473 1 0.441] If True - The index is valid and we are still attempting the quest.
      ; If False - We have just hit the last note in the array and the quest is complete
      (Utilities|Map|Add (Variables|Quest-BP|GetPlayerQuestTracking) _playeronkey (+ _questprogress 1))
      (if (Utilities|Array|IsValidIndex (Variables|Quest-BP|GetQuestArray) (Utilities|Map|Find (Variables|Quest-BP|GetPlayerQuestTracking) _playeronkey))
        ;[0.2967 1 0.2776] Clear or add data in PianoRunnerData
        (Variables|Quest-BP|SetQuestProgressIndex (+ _questprogress 1))
        (Development|PrintString "Correct Key was hit" (Variables|DebugMode|GetIsDebugModeOn) (Variables|DebugMode|GetIsDebugModeOn) "(R=0.131509,G=1.000000,B=0.000000,A=1.000000)")
        (Development|PrintString (Utilities|String|BuildString(Integer) "" "Progress Index = " (+ _questprogress 1)) (Variables|DebugMode|GetIsDebugModeOn) (Variables|DebugMode|GetIsDebugModeOn) "(R=0.390931,G=1.000000,B=0.061517,A=1.000000)")
        ;[0.3728 1 0.3948] Set enum on key so it knows what color to play
        (Variables|/Game/Athena/Items/QuestInteractables/BigPiano/Prop_Quest_PianoKey.Prop_Quest_PianoKey_C|SetColorSwap "NewEnumerator1" (Variables|Quest-BP|GetKeyHit))
        (else
          (Development|PrintString "Quest Complete!" (Variables|DebugMode|GetIsDebugModeOn) (Variables|DebugMode|GetIsDebugModeOn) "(R=1.000000,G=0.198145,B=0.000000,A=1.000000)" 3.0)
          (Variables|/Game/Athena/Items/QuestInteractables/BigPiano/Prop_Quest_PianoKey.Prop_Quest_PianoKey_C|SetColorSwap "NewEnumerator1" (Variables|Quest-BP|GetKeyHit))
          (Utilities|FlowControl|Delay (Variables|Default|GetDelayBeforeQuestFX))
          (CallFunction|AwardQuestProgress)
          (CallFunction|PlaySongBack)))
      (elif (Variables|Quest-BP|GetbIsFirstKey)
        ;[0.473 1 0.441] Set players quest progress to 1 because while it was the incorrect key, they have started over again by hitting the first note
        (Utilities|Map|Add (Variables|Quest-BP|GetPlayerQuestTracking) (Variables|Quest-BP|GetPlayerOnKey) 1)
        (Development|PrintString "Progress Index = 1" (Variables|DebugMode|GetIsDebugModeOn) (Variables|DebugMode|GetIsDebugModeOn) "(R=0.990837,G=1.000000,B=0.295815,A=1.000000)")
        ;[0.525 1 0.4583] Let the key know what color to play
        (Variables|/Game/Athena/Items/QuestInteractables/BigPiano/Prop_Quest_PianoKey.Prop_Quest_PianoKey_C|SetColorSwap "NewEnumerator1" (Variables|Quest-BP|GetKeyHit))
        (else
          ;[0.473 1 0.441] Reset quest structure
          (Utilities|Map|Remove (Variables|Quest-BP|GetPlayerQuestTracking) (Variables|Quest-BP|GetPlayerOnKey))
          (Development|PrintString "Progress Index = 0" (Variables|DebugMode|GetIsDebugModeOn) (Variables|DebugMode|GetIsDebugModeOn) "(R=1.000000,G=0.009744,B=0.000000,A=1.000000)")
          (Variables|/Game/Athena/Items/QuestInteractables/BigPiano/Prop_Quest_PianoKey.Prop_Quest_PianoKey_C|SetColorSwap "NewEnumerator2" (Variables|Quest-BP|GetKeyHit)))))))

(event Custom|AwardQuestProgress
  (bind _playeronkey (Variables|Quest-BP|GetPlayerOnKey))
  (if (Variables|Quest-Backend|GetbIsInPlayground)
    (CallFunction|ClearPianoQuestProgress)
    (else
      ;[0.4826 1 0.3708] Award Quest Progress
      (Development|PrintString (Utilities|String|Append "Length of Runner Array = " (Utilities|String|ToString(Integer) (Utilities|Array|Length (Variables|Quest-BP|GetContributedPlayers)))) (Variables|DebugMode|GetIsDebugModeOn) (Variables|DebugMode|GetIsDebugModeOn))
      (Class|AthenaQuestFunctionLibrary|AthenaGiveQuestUpdatetoPlayers :InQuestItemRef (Variables|Quest-Backend|GetQuest%20Item) :ObjectiveBackendName (Variables|Quest-Backend|GetQuestBackendObjectiveName) :PlayerControllersForUpdate (Utilities|Array|MakeArray _playeronkey))
      (Utilities|Map|Remove (Variables|Quest-BP|GetPlayerQuestTracking) _playeronkey)
      (Development|PrintString "Celebration time" (Variables|DebugMode|GetIsDebugModeOn) (Variables|DebugMode|GetIsDebugModeOn))
      (CallFunction|ClearPianoQuestProgress))))

(event Custom|ClearPianoQuestProgress
  ;[0.473 1 0.441] Set the array of players to a new empty array & set the progression index to 0
  (Utilities|Array|Clear (Variables|Quest-BP|GetContributedPlayers))
  (Variables|Quest-BP|SetQuestProgressIndex 0))

(event Custom|PlaySuccessFX
  (params (Controller :type Fort%20Player%20Controller%20Object%20Reference))
  (bind _returnvalue (Game|Player|GetPlayerPawn Controller))
  ;[0.5694 1 0.5527] Quest Success FX
  (Utilities|IsValid Controller
    (:Is%20Valid
      (Effects|SpawnFXSystemAttached (Variables|Quest-BP|GetSuccessFX) (Variables|/Script/Engine.Character|GetMesh _returnvalue))
      (Audio|PlaySoundatLocation (Variables|Quest-BP|GetSuccessSound) (Transformation|GetActorLocation _returnvalue)) ; Quest Success Sound
    )
    (:Is%20Not%20Valid
      (Development|PrintString "Player controller not valid" (Variables|DebugMode|GetIsDebugModeOn) (Variables|DebugMode|GetIsDebugModeOn)))))

(event Custom|PlaySongBack
  ;[0.6008 0.9501 1 1 50] SONG PLAYBACK FUNCTIONALITY
  ;;[0.5141 1 0.5219] CUSTOM EVENT -> PlaySongBack
  (Networking|SwitchHasAuthority
    (:Remote
      (Variables|Quest-BP|SetbIsPlayingSongBack true)
      ;;[0.512 1 0.5012] Locks things from happening on Prop_Quest_PianoKey
      (Components|Timeline|SetPlayRate (Variables|Prop_Quest_FullPiano|GetOnTheFirstDayOfChristmas) (Variables|Songs|GetPlayRate))
      (switch Utilities|FlowControl|Switch|SwitchonEnum_BigPiano_Songs (Variables|Default|GetSongToPlay)
        (:NewEnumerator0
          (Variables|PlaybackComponents|SetPlaybackArray (Variables|Songs|GetSong_FirstDayOfChristmas))
          (Timeline|OnTheFirstDayOfChristmas PlayFromStart
            (:Finished
              _)
            (:OnTheFirstDayOfChristmas
              ;;[1 1 1] Play Note (Same for all)
              (CallFunction|PlayNote))))
        (:NewEnumerator1
          (Variables|PlaybackComponents|SetPlaybackArray (Variables|Songs|GetSong_DefaultDance))
          (Timeline|DefaultDance PlayFromStart
            (:Finished
              _)
            (:DefaultDance
              _)))
        (:NewEnumerator2
          (Variables|PlaybackComponents|SetPlaybackArray (Variables|Songs|GetSong_JingleBells))
          (Variables|PlaybackComponents|SetPlaybackArray2 (Variables|Songs|GetSong_JingleBells2))
          (Timeline|JingleBells PlayFromStart
            (:Finished
              _)
            (:JingleBells
              _)))
        (:NewEnumerator3
          (Variables|PlaybackComponents|SetPlaybackArray (Variables|Songs|GetSong_BasicThing))
          (Timeline|Basic PlayFromStart
            (:Finished
              _)
            (:BasicCThing
              _))))
      ;;[1 1 1] Finished (Same for all)
      (Variables|PlaybackComponents|SetPlaybackIndex2 0)
      (Variables|PlaybackComponents|SetPlaybackIndex 0)
      (Variables|Quest-BP|SetbIsPlayingSongBack false))))

(event Custom|PlayNote
  (bind _playbackindex (Variables|PlaybackComponents|GetPlaybackIndex))
  (bind _playbackarray (Variables|PlaybackComponents|GetPlaybackArray))
  (bind _output (Utilities|Array|Get(acopy) _playbackarray _playbackindex))
  ;[0.5432 1 0.4965] If not a valid index then we've played back all the notes
  (if (Utilities|Array|IsValidIndex (Variables|PlaybackComponents|GetPlaybackArray) (Variables|PlaybackComponents|GetPlaybackIndex))
    ;[0.4943 1 0.5143] Find the key object with the name of our current index. if true then play its event "ReplicateVisuals" which plays its sound & visuals
    (bind _array_element (Utilities|Array|ForEachLoopwithBreak (Variables|Components-ChildKeyActor/ObjectReferences|GetKeyObjectReferenceArray))
      (:LoopBody
        (if (Utilities|String|EqualExactly(String) (Variables|/Game/Athena/Items/QuestInteractables/BigPiano/Prop_Quest_PianoKey.Prop_Quest_PianoKey_C|GetKeyName _array_element) _output)
          _))
      (:Completed
        _))
    (if (Utilities|String|EqualExactly(String) (Variables|/Game/Athena/Items/QuestInteractables/BigPiano/Prop_Quest_PianoKey.Prop_Quest_PianoKey_C|GetKeyName _array_element) _output)
      (Class|PropQuestPianoKey|ReplicateVisuals _array_element)
      ;[0.4453 1 0.4849] Up our jank index and set our time between notes reference by grabbing the value from the array
      (Math|Integer|IncrementInt (Variables|PlaybackComponents|GetPlaybackIndex)))))

(event Custom|PlayNote2
  (bind _playbackindex2 (Variables|PlaybackComponents|GetPlaybackIndex2))
  (bind _playbackarray2 (Variables|PlaybackComponents|GetPlaybackArray2))
  (bind _output (Utilities|Array|Get(acopy) _playbackarray2 _playbackindex2))
  ;[0.5432 1 0.4965] If not a valid index then we've played back all the notes
  (if (Utilities|Array|IsValidIndex (Variables|PlaybackComponents|GetPlaybackArray2) (Variables|PlaybackComponents|GetPlaybackIndex2))
    ;[0.4943 1 0.5143] Find the key object with the name of our current index. if true then play its event "ReplicateVisuals" which plays its sound & visuals
    (bind _array_element (Utilities|Array|ForEachLoopwithBreak (Variables|Components-ChildKeyActor/ObjectReferences|GetKeyObjectReferenceArray))
      (:LoopBody
        (if (Utilities|String|EqualExactly(String) (Variables|/Game/Athena/Items/QuestInteractables/BigPiano/Prop_Quest_PianoKey.Prop_Quest_PianoKey_C|GetKeyName _array_element) _output)
          _))
      (:Completed
        _))
    (if (Utilities|String|EqualExactly(String) (Variables|/Game/Athena/Items/QuestInteractables/BigPiano/Prop_Quest_PianoKey.Prop_Quest_PianoKey_C|GetKeyName _array_element) _output)
      (Class|PropQuestPianoKey|ReplicateVisuals _array_element)
      ;[0.4453 1 0.4849] Up our jank index and set our time between notes reference by grabbing the value from the array
      (Math|Integer|IncrementInt (Variables|PlaybackComponents|GetPlaybackIndex2)))))

(fn AddPlayerToQuestStructure
  (params (PlayerController :type Fort%20Player%20Controller%20Object%20Reference) (KeyPressed :type String))
  (bind _localplayerfound (LocalVariables|GetLocalPlayerFound))
  (bind _localplayer (LocalVariables|GetLocalPlayer))
  (bind _output_get (LocalVariables|SetLocalPlayer PlayerController))
  (if (== (Variables|Quest-BP|GetQuestProgressIndex) 0)
    (bind _returnvalue (Utilities|Array|Add (Variables|Quest-BP|GetContributedPlayers) _output_get))
    (return false _returnvalue 0)
    (else
      (bind (_array_element _array_index) (Utilities|Array|ForEachLoopwithBreak (Variables|Quest-BP|GetContributedPlayers))
        (:LoopBody
          ;[1 1 1] If the player is found, set index found and break loop
          (bind _output_get_2 (LocalVariables|SetLocalPlayerFound (Utilities|Equal(Object) (LocalVariables|GetLocalPlayer) _array_element)))
          (if _output_get_2
            (LocalVariables|SetLocalPlayerIndex _array_index)))
        (:Completed
          _))
      ;[1 1 1] Player was found, give the goods
      (if _localplayerfound
        (return _localplayerfound (LocalVariables|GetLocalPlayerIndex) (LocalVariables|GetLocalPlayer))
        (else
          (bind _returnvalue_2 (Utilities|Array|Add (Variables|Quest-BP|GetContributedPlayers) _localplayer))
          (return false _returnvalue_2 _localplayer))))))

(fn BlueprintCanInteract
  (params (InteractingPawn :type Fort%20Pawn%20Object%20Reference) (InteractionBeingAttempted :type EInteractionBeingAttempted%20Enum :default "FirstInteraction") (InteractionType :type TInteractionType%20Enum :default "IT_NoInteraction"))
  (return false))

(fn DidPlayerCompleteQuestAlready
  (params (LocalPlayerOnKey :type Fort%20Player%20Controller%20Object%20Reference))
  (bind _array_element (Utilities|Array|ForEachLoopwithBreak (Variables|Default|GetPlayersWhoCompletedQuest))
    (:LoopBody
      (if (Utilities|Equal(Object) _array_element LocalPlayerOnKey)
        (LocalVariables|SetbLocalAlreadyCompletedQuest true)))
    (:Completed
      _))
  (return (LocalVariables|GetbLocalAlreadyCompletedQuest)))

(fn IsHitKeyCorrect
  (bind _playeronkey (Variables|Quest-BP|GetPlayerOnKey))
  (bind _playerquesttracking (Variables|Quest-BP|GetPlayerQuestTracking))
  (bind _value (Utilities|Map|Find _playerquesttracking _playeronkey))
  (bind _questarray (Variables|Quest-BP|GetQuestArray))
  (if (Utilities|Array|IsValidIndex _questarray _value)
    (return _value (Utilities|String|EqualExactly(String) (Utilities|Array|Get(aref) _questarray _value) (Variables|Quest-BP|GetKeyNameSteppedOn)))
    (else
      (return 0 false))))

(fn IsKeyPartOfQuest
  (params (Key :type String))
  ;[0.451 1 0.4177] Loop through the quest array and compare the key hit to each index, if it matches, set IsInQuest=True, if it matched at the first index, set IsFirst=True
  (bind (_array_element _array_index) (Utilities|Array|ForEachLoopwithBreak (Variables|Quest-BP|GetQuestArray))
    (:LoopBody
      (if (Utilities|String|EqualExactly(String) Key _array_element)
        (LocalVariables|SetIsInQuest true)
        ;[0.5671 1 0.4794] If when IsInQuest was set to true, was it the first index
        (if (== _array_index 0)
          (LocalVariables|SetIsFirst true))))
    (:Completed
      _))
  (return (LocalVariables|GetIsInQuest) (LocalVariables|GetIsFirst) Key))

(fn SetSoundArray
  (params (KeyArrayToAssign :type Array%20of%20Prop%20Quest%20Piano%20Key%20Object%20References))
  ;[0.3715 1 0.2292] Take each key objects audio component and set the int param "KeysIndex" to its proper int index (makes the keys sound independent of quest completion)
  (for (_array_element _array_index) KeyArrayToAssign
    (Audio|Parameter|SetIntegerParameter (Variables|/Game/Athena/Items/QuestInteractables/BigPiano/Prop_Quest_PianoKey.Prop_Quest_PianoKey_C|GetKeySound _array_element) "KeysIndex" _array_index :!class /Script/AudioExtensions.AudioParameterControllerInterface)))

(fn ConstructionScript
  ;[0.5871 1 0.5667] Create array of key objects so we don't have to cast it all the time
  (bind _output_get (Variables|Components-ChildKeyActor/ObjectReferences|SetChildKeyArray (Utilities|Array|MakeArray (Variables|Default|GetC1) (Variables|Default|GetD1) (Variables|Default|GetE1) (Variables|Default|GetF1) (Variables|Default|GetG1) (Variables|Default|GetA1) (Variables|Default|GetB1) (Variables|Default|GetC2) (Variables|Default|GetD2) (Variables|Default|GetE2) (Variables|Default|GetF2) (Variables|Default|GetG2) (Variables|Default|GetA2) (Variables|Default|GetB2) (Variables|Default|GetC#1) (Variables|Default|GetD#1) (Variables|Default|GetF#1) (Variables|Default|GetG#1) (Variables|Default|GetA#1) (Variables|Default|GetC#2) (Variables|Default|GetD#2) (Variables|Default|GetF#2) (Variables|Default|GetG#2) (Variables|Default|GetA#2))))
  (Utilities|FlowControl|Sequence
    (:then_0
      (for (_array_element _array_index) _output_get
        (bind _asprop_quest_piano_key (Utilities|Casting|CastToProp_Quest_PianoKey (Variables|/Script/Engine.ChildActorComponent|GetChildActor _array_element)))
        (Utilities|Map|Add (Variables|Quest-BP|GetChild/IndexMap) _asprop_quest_piano_key _array_index)
        (Utilities|Array|Add (Variables|Components-ChildKeyActor/ObjectReferences|GetKeyObjectReferenceArray) _asprop_quest_piano_key)))
    (:then_1
      ;[0.5051 1 0.5936] Wait for objects to be put in the array, then this function sets their int params
      (CallFunction|SetSoundArray (Variables|Components-ChildKeyActor/ObjectReferences|GetKeyObjectReferenceArray)))
    (:then_2
      ;[0.4711 1 0.4898] 13th index is the last white key
      (for _index (range 14)
        ;[0.4711 1 0.4898] Easy way to change spacing of the keys. Have not done it for black keys because I'm a dingus.
        (Transformation|SetRelativeLocation (Utilities|Array|Get(acopy) (Variables|Components-ChildKeyActor/ObjectReferences|GetChildKeyArray) _index) (Math|Vector|MakeVector (* _index 103.0)) false false)))
    (:then_3
      (for (_array_element_2 _array_index_2) (Variables|Components-ChildKeyActor/ObjectReferences|GetKeyObjectReferenceArray)
        (Rendering|Material|SetVectorParameterValue (Variables|/Game/Athena/Items/QuestInteractables/BigPiano/Prop_Quest_PianoKey.Prop_Quest_PianoKey_C|GetM_Dynamic_Index_1 _array_element_2) "UV Offset" (Math|Color|MakeColor (* (Math|Integer|%25%28Integer%29 _array_index_2 5) 0.2) (* (/ _array_index_2 5) 0.2) 0.0 0.0) :!class /Script/Engine.MaterialInstanceDynamic)))))
