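; Small child variant of the full piano quest prop, a reduced version of its parent Blueprint.
; The construction script runs the parent's setup, then loops over an array of the twelve octave-2 note components, naturals and sharps, and destroys each one.
; The BeginPlay, overlap and Tick events add no logic of their own.
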(event EventBeginPlay
  (|CallParentFunction) ; This node is disabled and will not be called.\nDrag off pins to build functionality.
)

(event Collision|EventActorBeginOverlap
  (params (OtherActor :type Actor%20Object%20Reference)))

(event EventTick
  (params (DeltaSeconds :type Float%20%28single-precision%29)))

(fn ConstructionScript
  (|CallParentFunction)
  (for _array_element (Utilities|Array|MakeArray (Variables|Default|GetC2) (Variables|Default|GetD2) (Variables|Default|GetE2) (Variables|Default|GetF2) (Variables|Default|GetG2) (Variables|Default|GetA2) (Variables|Default|GetB2) (Variables|Default|GetC#2) (Variables|Default|GetD#2) (Variables|Default|GetF#2) (Variables|Default|GetG#2) (Variables|Default|GetA#2))
    (Components|DestroyComponent _array_element)))
