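; Single key of the quest big piano, a child actor that plays a note and glows when stepped on.
; Server-side overlap, debounced by a short delay, reports the key and player to the parent FullPiano unless song playback is running, then calls ReplicateVisuals.
; Glow timelines drive a dynamic material, choosing default, correct or incorrect color from ColorSwap.
; Construction configures black or white key materials, offsets black-key collision and swaps in a mobile sound.
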(event EventBeginPlay)

(event OnComponentBeginOverlap(Collision)
  (params (OverlappedComponent :type Primitive%20Component%20Object%20Reference) (OtherActor :type Actor%20Object%20Reference) (OtherComp :type Primitive%20Component%20Object%20Reference) (OtherBodyIndex :type Integer) (bFromSweep :type Boolean) (SweepResult :type Hit%20Result%20Structure%20%28by%20ref%29))
  (bind _returnvalue (Actor|GetParentActor))
  ;[1 1 1] Track quest progression on server only?
  (Networking|SwitchHasAuthority
    (:Authority
      (if (Variables|Default|GetbCanPlay)
        (Variables|Default|SetbCanPlay false)
        ;[1 1 1] Because I duplicated the FullPiano blueprint for the smaller piano, I have to check if we're casting to use functionality from the small piano, or the big one. I know it's terrible.
        (Utilities|FlowControl|Sequence
          (:then_0
            (if (Actor|ActorHasTag _returnvalue "FullPiano")
              (bind _asprop_quest_full_piano (Utilities|Casting|CastToProp_Quest_FullPiano _returnvalue))
              (if (not (Variables|/Game/Athena/Items/QuestInteractables/BigPiano/Prop_Quest_FullPiano.Prop_Quest_FullPiano_C|GetbIsPlayingSongBack _asprop_quest_full_piano))
                (Class|PropQuestFullPiano|KeySteppedOn _asprop_quest_full_piano self (Fortnite|GetFortPlayerControllerfromActor OtherActor))
                (CallFunction|ReplicateVisuals))))
          (:then_1
            (Utilities|FlowControl|Delay 0.2)
            (Variables|Default|SetbCanPlay true)))))))

(event OnComponentEndOverlap(Collision)
  (params (OverlappedComponent :type Primitive%20Component%20Object%20Reference) (OtherActor :type Actor%20Object%20Reference) (OtherComp :type Primitive%20Component%20Object%20Reference) (OtherBodyIndex :type Integer))
  (Networking|SwitchHasAuthority))

(event Custom|ReplicateVisuals
  (CallFunction|PlaySound)
  (CallFunction|KeyGlow))

(event Custom|KeyGlow
  ;[1 1 1] Changed at the end of quest tracking logic in FullPiano BP
  (switch Utilities|FlowControl|Switch|SwitchonBigPianoColors_Enum (Variables|Default|GetColorSwap)
    (:NewEnumerator0
      ;[1 1 1] Custom color curves for vect params on M_PianoKey_Parent or MI's of it.
      (bind (_direction _glowintensity) (Timeline|GlowColor PlayFromStart)
        (:Update
          ;;[0.1354 0.5504 1] Default
          (Rendering|Material|SetVectorParameterValue (Variables|Visuals|GetM_Dynamic_Index_0) "GlowColor" _glowintensity :!class /Script/Engine.MaterialInstanceDynamic))
        (:Finished
          _)))
    (:NewEnumerator1
      (bind (_direction_2 _glowintensity_2) (Timeline|Timeline_0 PlayFromStart)
        (:Update
          ;;[0.3437 1 0.4089] Correct
          (Rendering|Material|SetVectorParameterValue (Variables|Visuals|GetM_Dynamic_Index_0) "GlowColor" _glowintensity_2 :!class /Script/Engine.MaterialInstanceDynamic))
        (:Finished
          _)))
    (:NewEnumerator2
      (bind (_direction_3 _glowintensity_3) (Timeline|Timeline_1 PlayFromStart)
        (:Update
          ;;[1 0.2028 0.2925] Incorrect
          (Rendering|Material|SetVectorParameterValue (Variables|Visuals|GetM_Dynamic_Index_0) "GlowColor" _glowintensity_3 :!class /Script/Engine.MaterialInstanceDynamic))
        (:Finished
          _))))
  ;[1 1 1] When the color is done updating go back to assuming normal color on hit
  (Variables|Default|SetColorSwap "NewEnumerator0")
  (Rendering|Material|SetVectorParameterValue (Variables|Visuals|GetM_Dynamic_Index_0) "GlowColor" "(R=0.000000,G=0.000000,B=0.000000,A=0.000000)" :!class /Script/Engine.MaterialInstanceDynamic))

(event Custom|PlaySound
  (Audio|Components|Audio|Play (Variables|Default|GetKeySound) 0.0))

(fn OnRep_bSteppedOn
  (Networking|SwitchHasAuthority
    (:Remote
      (if (Variables|Default|GetbSteppedOn)
        (CallFunction|PlaySound)
        (CallFunction|KeyGlow)))))

(fn OnRep_ColorSwap)

(fn OnRep_GetColor)

(fn SetupMaterialInstances
  (bind _keymesh (Variables|Default|GetKeyMesh))
  (bind _keymesh_2 (Variables|Default|GetKeyMesh))
  (bind _keymesh_3 (Variables|Default|GetKeyMesh))
  ;[1 1 1] Create Material Instance For Selfs Mesh
  (if (Variables|Visuals|GetKeyIsBlack)
    ;[1 1 1] Set Material Based On Instance Of Boolean "KeyIsBlack"
    (bind _returnvalue (Rendering|Material|CreateDynamicMaterialInstance _keymesh 0 "/Game/Athena/Items/QuestInteractables/BigPiano/Materials/MI_PianoKey_Black.MI_PianoKey_Black" "None" :!class /Script/Engine.PrimitiveComponent))
    (bind _output_get (Variables|Visuals|SetM_Dynamic_Index_0 _returnvalue))
    (Rendering|Material|SetMaterial _keymesh_3 0 _output_get)
    (bind _returnvalue_2 (Rendering|Material|CreateDynamicMaterialInstance _keymesh_3 1 "/Game/Athena/Items/QuestInteractables/BigPiano/Materials/M_KeysText.M_KeysText" "None" :!class /Script/Engine.PrimitiveComponent))
    (bind _output_get_2 (Variables|Visuals|SetM_Dynamic_Index_1 _returnvalue_2))
    (Rendering|Material|SetVectorParameterValue _output_get_2 "letter_tint" "(R=1.000000,G=1.000000,B=1.000000,A=1.000000)" :!class /Script/Engine.MaterialInstanceDynamic)
    (else
      (bind _returnvalue_3 (Rendering|Material|CreateDynamicMaterialInstance _keymesh 0 "/Game/Athena/Items/QuestInteractables/BigPiano/Materials/MI_PianoKey_White.MI_PianoKey_White" "None" :!class /Script/Engine.PrimitiveComponent))
      (bind _output_get_3 (Variables|Visuals|SetM_Dynamic_Index_0 _returnvalue_3))
      (Rendering|Material|SetMaterial _keymesh_2 0 _output_get_3)
      (bind _returnvalue_4 (Rendering|Material|CreateDynamicMaterialInstance _keymesh_2 1 "/Game/Athena/Items/QuestInteractables/BigPiano/Materials/M_KeysText.M_KeysText" "None" :!class /Script/Engine.PrimitiveComponent))
      (bind _output_get_4 (Variables|Visuals|SetM_Dynamic_Index_1 _returnvalue_4))
      (Rendering|Material|SetVectorParameterValue _output_get_4 "letter_tint" "(R=0.000000,G=0.000000,B=0.000000,A=1.000000)" :!class /Script/Engine.MaterialInstanceDynamic))))

(fn ConstructionScript
  (Utilities|FlowControl|Sequence
    (:then_0
      ;[1 1 1] Comment
      (if (Variables|Visuals|GetKeyIsBlack)
        (Components|StaticMesh|SetStaticMesh (Variables|Default|GetKeyMesh) "/Game/Athena/Items/QuestInteractables/BigPiano/Meshes/SM_big_piano_key_white.SM_big_piano_key_white")))
    (:then_1
      ;[1 1 1] Comment
      (CallFunction|SetupMaterialInstances))
    (:then_2
      ;[1 1 1] jank small adjustments for the black keys
      (if (Variables|Visuals|GetKeyIsBlack)
        (Transformation|SetRelativeLocation (Variables|Default|GetCollision) (Math|Vector|MakeVector (+ (.x (Variables|/Script/Engine.SceneComponent|GetRelativeLocation (Variables|Default|GetCollision))) -7.0) (+ (.y (Variables|/Script/Engine.SceneComponent|GetRelativeLocation (Variables|Default|GetCollision))) 0.0) (+ (.z (Variables|/Script/Engine.SceneComponent|GetRelativeLocation (Variables|Default|GetCollision))) 0.0)) false false)))
    (:then_3
      ;[1 1 1] Change the sound cue if we're mobile
      (Networking|SwitchHasAuthority
        (:Remote
          (if (TouchPlatform|IsMobileGame)
            (Audio|Components|Audio|SetSound (Variables|Default|GetKeySound) "/Game/Athena/Sounds/Quests/BigPiano/New_Wavs/Piano_Note_Parent_Mobile.Piano_Note_Parent_Mobile")
            (else
              (Development|PrintString "we aint mobile" false false))))))))
